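; BP_S19_GOW_Cover is a Season 19 Battle Pass quest prop actor that stays in the world only while its calendar event is active.
; On BeginPlay it runs the parent logic, waits a fixed delay, then looks up the calendar event named by its ConsolidatedQuestComponent and destroys itself if inactive.
; Overlap, tick and failed-visuals events are empty stubs, and the construction script only calls the parent.
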(event EventBeginPlay
  (|CallParentFunction)
  (Utilities|FlowControl|Delay 5.0)
  (bind _biseventactive (CalendarEvents|GetCalendarEventInformation (Variables|/Script/FortniteGame.ConsolidatedQuestComponent|GetCalendarEnableEvent (Variables|BuildingGameplayActorPropQuest|GetConsolidatedQuestComponent))))
  (if (not _biseventactive)
    (Actor|DestroyActor)))

(event Collision|EventActorBeginOverlap
  (params (OtherActor :type Actor%20Object%20Reference)))

(event EventTick
  (params (DeltaSeconds :type Float%20%28single-precision%29)))

(event OnCustomSetVisualstoFailed(ConsolidatedQuestComponent)
  (params (PlayerController :type Fort%20Player%20Controller%20Object%20Reference)))

(fn ConstructionScript
  (|CallParentFunction))
